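# Niagara storm weather effect with near and far rain, distant rain sheets and lightning, likely for the Athena storm visuals.
# GPU rain emitters spawn two groups, one with loop-animated offsets, stretch sprites by speed and randomly enable zero-bounce collision; far rain adds curl noise until impact.
# CPU emitters add large falling sheet sprites and randomly timed lightning meshes with curve-driven flicker and lights.
# User.StormIntensity is declared but unused in the shown modules.
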NiagaraSystem /FNGameplayCues/VisualThreatManager/StormVisuals/Particles/NS_Storm_Rain_Athena_Impacts
dependencies {
 usedDataInterfaces [&NiagaraDataInterfaceCurve, &NiagaraDataInterfaceVector2DCurve]
 usedDynamicInputs []
 usedModules [&SystemState]
}
emitters {
 Distant_Rain_Sheet {
  emitterUpdateScript [
   &EmitterState()
   &SpawnRate(SpawnRate=&RandomRangeFloat(Minimum=7.5, Maximum=12), Use Spawn Probability=false)
  ]
  particleSpawnScript [
   &InitializeParticle(
    Color Maximum=<rgba 0.6979 0.77 1 0.4>
    Color Minimum=<rgba 0.6979 0.77 1 0.3>
    Color Mode=::NewEnumerator2
    Lifetime Max=1.5
    Lifetime Min=0.75
    Lifetime Mode=::NewEnumerator1
    Sprite Size Mode=::NewEnumerator1
    Uniform Sprite Size Max=6500
    Uniform Sprite Size Min=5000
    UsePositionOffset=false
    Write Material Random=true
   )
   &ShapeLocation(
    Cylinder Height=0
    Cylinder Radius=&RandomRangeFloat(Minimum=2400, Maximum=5000)
    MaskSpawnGroup=false
    Offset=&RandomRangeVector(Minimum=<0 0 2500>, Maximum=<0 0 4000>)
    Offset Mode=::NewEnumerator0
    Override Seed=false
    Shape Primitive=::NewEnumerator1
    Surface Only=true
    UseOwnerScale=true
   )
   &AddVelocity(Velocity=&RandomRangeVector(Minimum=<0 0 -1600>, Maximum=<0 0 -1900>))
  ]
  particleUpdateScript [
   &ParticleState()
   &ScaleSpriteSize(
    Non-Uniform Curve Sprite Scale=DataInterface() {propertyValues {
      XCurve {keys [{tangentMode=RCTM_User, time=0, value=0.6142}, {tangentMode=RCTM_User, time=1, value=0.75}]}
      YCurve {keys [{tangentMode=RCTM_User, time=0, value=1}, {tangentMode=RCTM_User, time=1, value=2}]}
     }}
    Scale Sprite Size Mode=::NewEnumerator3
   )
   &SolveForcesAndVelocity(Clamp Velocity=false, Limit Acceleration=false)
   &ScaleColor(
    Scale Alpha=&FloatFromCurve(FloatCurve=DataInterface() {propertyValues {Curve {
        defaultValue=1
        keys [
         {interpMode=RCIM_Cubic, tangentMode=RCTM_User, time=0, value=0}
         {interpMode=RCIM_Cubic, tangentMode=RCTM_User, time=0.5, value=1}
         {tangentMode=RCTM_User, time=1, value=0}
        ]
       }}})
    ScaleA=true
    ScaleRGB=false
   )
   &DynamicMaterialParameters(
    Index 0 Param 1=-0.7
    Index 0 Param 2=1
    Index 0 Param 3=1
    Index 0 Param 4=1
    Module.Index0Float0Write=true
    Module.Index0Float1Write=true
    Module.Index0Float2Write=true
    Module.Index0Float3Write=true
   )
  ]
  rendererClasses [&NiagaraSpriteRendererProperties]
  renderers [{rendererClass=&NiagaraSpriteRendererProperties}]
 }
 Far_Rain_Impact {
  simTarget=GPUComputeSim
  emitterUpdateScript [
   &EmitterState(Life Cycle Mode=::NewEnumerator1, Loop Duration=2, UseLoopDelay=false)
   &SpawnRate(SpawnRate=3000, Use Spawn Probability=false)
   &SpawnRate(SpawnGroup=1, SpawnRate=700, Use Spawn Probability=false) {module=SpawnRate001}
  ]
  particleSpawnScript [
   &InitializeParticle(
    Color=&MaskLinearColorBySpawnGroup(Masked Linear Color=<rgba 2.511 2.1875 4 1.6>, Passthrough Linear Color=<rgba 1.3582 1.2188 2 2>)
    Lifetime Min=1.8
    Lifetime Mode=::NewEnumerator1
    Sprite Size Mode=::NewEnumerator3
    Sprite UV Mode=::NewEnumerator3
    Uniform Sprite Size=&MaskFloatBySpawnGroup(Masked Float=3, Passthrough Float=&RandomRangeFloat(Minimum=3, Maximum=5))
    UsePositionOffset=false
    Write Material Random=true
   )
   &ShapeLocation(
    Cylinder Height=0
    Cylinder Radius=&RandomRangeFloat(Minimum=2400, Maximum=6000)
    MaskSpawnGroup=true
    Offset=<0 0 1500>
    Offset Mode=::NewEnumerator0
    Override Seed=false
    Shape Primitive=::NewEnumerator1
    Surface Only=true
    Use Endcaps In Surface Only Mode=false
    UseOwnerScale=true
   )
   &ShapeLocation(
    Cylinder Height=0
    Cylinder Radius=&RandomRangeFloat(Minimum=3400, Maximum=8000)
    MaskSpawnGroup=true
    Offset=&MakeVector(
     X=0
     Y=0
     Z=&FloatFromCurve(FloatCurve=DataInterface() {propertyValues {Curve {keys [<at 0 3250>,<at 0.5 4000>,<at 1 2250>]}}}, CurveIndex=$Emitter.NormalizedLoopAge)
    )
    Offset Mode=::NewEnumerator0
    Override Seed=false
    Shape Primitive=::NewEnumerator1
    Spawn Group Mask=1
    Surface Only=true
    UseOwnerScale=true
   ) {module=ShapeLocation001}
   &AddVelocity(Velocity=&RandomRangeVector(Minimum=<300 150 -1925>, Maximum=<1200 -256 -2100>))
  ]
  particleUpdateScript [
   &ParticleState()
   &Collision(
    Age Colliding Particles=false
    Collision Enabled=&RandomBool()
    Friction=1
    Friction During a Bounce=1
    GPU Collision Type=::NewEnumerator2
    KillOnCollision=false
    Module.Control Rest State Via Surface Normal=false
    Module.EnableMaxCollisionCount=false
    Module.Randomize Collision Normal Vector=false
    Restitution=0
   )
   &CurlNoiseForce(
    Bias Noise=false
    Noise Frequency=100
    Noise Quality / Cost=::NewEnumerator0
    Noise Strength=&MakeCustomFloatFromBool(Bool=$Particles.HasCollided, True Float=0, False Float=5500)
    Pan Noise=true
    Pan Noise Field=&RandomRangeVector(Minimum=<-1 -1 -1>, Maximum=<1 1 1>, Evaluation Type=::NewEnumerator1)
   )
   &SolveForcesAndVelocity(Clamp Velocity=false, Limit Acceleration=false)
   &ScaleSpriteSizeBySpeed(Max Scale Factor=<1 12>)
   &ScaleColor(
    Scale Alpha=&FloatFromCurve(FloatCurve=DataInterface() {propertyValues {Curve {keys [
         <at 0 0 RCIM_Cubic>
         {arriveTangentWeight=0.0465, interpMode=RCIM_Cubic, leaveTangentWeight=0.3604, tangentMode=RCTM_User, time=0.08, value=1}
         {arriveTangentWeight=0.3052, interpMode=RCIM_Cubic, leaveTangentWeight=0.1477, tangentMode=RCTM_User, time=0.7, value=1}
         <at 1 0 RCIM_Cubic>
        ]}}})
    ScaleA=true
    ScaleRGB=false
   )
  ]
  rendererClasses [&NiagaraSpriteRendererProperties]
  renderers [{rendererClass=&NiagaraSpriteRendererProperties}]
 }
 Lightning_Strikes {
  emitterUpdateScript [
   &EmitterState(
    Life Cycle Mode=::NewEnumerator1
    Loop Duration=&RandomRangeFloat(Minimum=0.15, Maximum=1)
    Recalculate Duration Each Loop=true
    UseLoopDelay=false
   )
   &SpawnBurst_Instantaneous(Use Spawn Probability=false, UseLoopCountLimit=false)
  ]
  particleSpawnScript [
   &InitializeParticle(Lifetime=0.5, Mesh Scale Mode=::NewEnumerator3, Mesh Uniform Scale=50, UsePositionOffset=false, Write Material Random=true)
   &ShapeLocation(
    Cylinder Height=1000
    Cylinder Radius=&RandomRangeFloat(Minimum=5000, Maximum=50000)
    MaskSpawnGroup=false
    Offset=<0 0 -1000>
    Offset Mode=::NewEnumerator0
    Override Seed=false
    Shape Primitive=::NewEnumerator1
    Surface Only=true
    Use Endcaps In Surface Only Mode=false
    UseOwnerScale=true
   )
   &InitialMeshOrientation(
    Mesh Orientation Mode=::NewEnumerator2
    Orientation Vector=<0 0 1>
    Rotation=&RandomRangeVector(Minimum=<0 0 -0.1>, Maximum=<1 0 0.1>)
    Use Rotation Vector=true
   )
  ]
  particleUpdateScript [
   &ParticleState()
   &DynamicMaterialParameters(
    Index 0 Param 1=&FloatFromCurve(FloatCurve=DataInterface() {propertyValues {Curve {keys [
         {arriveTangent=-0.7436, interpMode=RCIM_Cubic, leaveTangent=-0.7436, tangentMode=RCTM_User, time=0, value=1}
         {arriveTangent=-0.489, interpMode=RCIM_Cubic, leaveTangent=-0.489, tangentMode=RCTM_User, time=0.37, value=0.5}
         {arriveTangent=-1.2768, interpMode=RCIM_Cubic, leaveTangent=-1.2768, tangentMode=RCTM_User, time=1, value=0}
        ]}}})
    Index 0 Param 2=&FloatFromCurve(FloatCurve=DataInterface() {propertyValues {Curve {keys [
         <at 0 0.15 RCIM_Cubic>
         {arriveTangent=6.9678, interpMode=RCIM_Cubic, leaveTangent=6.9678, tangentMode=RCTM_User, time=0.25, value=0.65}
         {arriveTangent=-7.8904, interpMode=RCIM_Cubic, leaveTangent=-7.8904, tangentMode=RCTM_User, time=0.3, value=3.15}
         {arriveTangent=-1.4941, interpMode=RCIM_Cubic, leaveTangent=-1.4941, tangentMode=RCTM_User, time=0.35, value=0.35}
         <at 1 0.05 RCIM_Cubic>
        ]}}})
    Index 0 Param 3=&FloatFromCurve(FloatCurve=DataInterface() {propertyValues {Curve {keys [
         <at 0.1 1 RCIM_Cubic>
         {arriveTangent=-1.2, interpMode=RCIM_Cubic, leaveTangent=-1.2, time=0.3, value=1.46}
         {arriveTangent=-4.64, interpMode=RCIM_Cubic, leaveTangent=-4.64, time=0.35, value=0.7}
         {arriveTangentWeight=0.1882, interpMode=RCIM_Cubic, leaveTangentWeight=0.4141, tangentMode=RCTM_User, time=0.55, value=0.3}
         {arriveTangent=8.0816, interpMode=RCIM_Cubic, leaveTangent=8.0816, tangentMode=RCTM_User, time=0.99, value=2.39}
        ]}}})
    Module.Index0Float0Write=true
    Module.Index0Float1Write=true
    Module.Index0Float2Write=true
    Module.Index0Float3Write=true
   )
   &ScaleColor(
    Scale Alpha=&FloatFromCurve(FloatCurve=DataInterface() {propertyValues {Curve {keys [<at 0.4 1>,<at 1 0>]}}})
    Scale RGB=&VectorFromFloat(Value=&FloatFromCurve(FloatCurve=DataInterface() {propertyValues {Curve {keys [
          <at 0.2 1 RCIM_Cubic>
          {arriveTangent=-12.7422, interpMode=RCIM_Cubic, leaveTangent=-12.7422, tangentMode=RCTM_User, time=0.25, value=1.98}
          {arriveTangentWeight=0.1106, interpMode=RCIM_Cubic, leaveTangentWeight=0.4424, tangentMode=RCTM_User, time=0.4, value=1}
          {arriveTangent=-2.7532, interpMode=RCIM_Cubic, leaveTangent=-2.7532, tangentMode=RCTM_User, time=1, value=0.5}
         ]}}}))
    ScaleA=true
    ScaleRGB=true
   )
   &Light_Attributes(
    Color=1
    Color Mode=0
    Falloff Exponent=1
    Light Color:LinearColor=&MultiplyLinearColor(
     Linear Color=<rgba 0.6747 0.0496 1 1>
     Scale Factor=&MakeVector4FromVectorAndFloat(
      "Vector (RGB)"=&VectorFromFloat(Value=&FloatFromCurve(FloatCurve=DataInterface() {propertyValues {Curve {keys [
            <at 0.2 1 RCIM_Cubic>
            {arriveTangent=-12.7422, interpMode=RCIM_Cubic, leaveTangent=-12.7422, tangentMode=RCTM_User, time=0.25, value=1.98}
            {arriveTangentWeight=0.1106, interpMode=RCIM_Cubic, leaveTangentWeight=0.4424, tangentMode=RCTM_User, time=0.4, value=1}
            {arriveTangent=-2.7532, interpMode=RCIM_Cubic, leaveTangent=-2.7532, tangentMode=RCTM_User, time=1, value=0.5}
           ]}}}))
      "Float (Alpha)"=&FloatFromCurve(FloatCurve=DataInterface() {propertyValues {Curve {keys [<at 0 1>,<at 1 0>]}}}, Scale Curve=50)
     )
    )
    Light Exponent=16
    Light Position Offset:Vector3f=<0 0 5000>
    Light Radius=40000
    Position=1
    Radius=1
    UsePositionOffset=true
   )
  ]
  rendererClasses [&NiagaraMeshRendererProperties, &NiagaraLightRendererProperties]
  renderers [
   {rendererClass=&NiagaraMeshRendererProperties, rendererIndex=0}
   {rendererClass=&NiagaraLightRendererProperties, rendererIndex=1}
  ]
 }
 Near_Rain_Impacts {
  simTarget=GPUComputeSim
  emitterUpdateScript [
   &EmitterState(Life Cycle Mode=::NewEnumerator1, Loop Duration=2, UseLoopDelay=false)
   &SpawnRate(SpawnRate=4250, Use Spawn Probability=false)
   &SpawnRate(SpawnGroup=1, SpawnRate=425, Use Spawn Probability=false) {module=SpawnRate001}
  ]
  particleSpawnScript [
   &InitializeParticle(
    Color=&MaskLinearColorBySpawnGroup(Masked Linear Color=<rgba 0.1878 0.137 0.4219 0.4>, Passthrough Linear Color=<rgba 0.6281 0.5529 0.974 1.2>)
    Lifetime Max=1
    Lifetime Min=0.8
    Lifetime Mode=::NewEnumerator1
    Sprite Size Max=&MaskVector2dBySpawnGroup(Masked Vector2D=<2 2>, Passthrough Vector2D=<2.5 3>, Spawn Group Mask=0)
    Sprite Size Min=&MaskVector2dBySpawnGroup(Masked Vector2D=<1.5 2>, Passthrough Vector2D=<2 2.2>, Spawn Group Mask=0)
    Sprite Size Mode=::NewEnumerator2
    Sprite UV Mode=::NewEnumerator3
    UsePositionOffset=false
    Write Material Random=true
   )
   &ShapeLocation(
    Cylinder Height=0
    Cylinder Radius=&RandomRangeFloat(Minimum=50, Maximum=2400)
    MaskSpawnGroup=true
    Offset=<0 0 1500>
    Offset Mode=::NewEnumerator0
    Override Seed=false
    Shape Primitive=::NewEnumerator1
    Surface Only=true
    Use Endcaps In Surface Only Mode=false
    UseOwnerScale=true
   )
   &ShapeLocation(
    Cylinder Height=0
    Cylinder Radius=&RandomRangeFloat(Minimum=150, Maximum=300)
    MaskSpawnGroup=true
    Offset=&MakeVector(
     X=&RandomRangeFloat(
      Minimum=&FloatFromCurve(FloatCurve=DataInterface() {propertyValues {Curve {keys [<at 0 125>,<at 0.5 650>,<at 1 45>]}}}, CurveIndex=$Emitter.NormalizedLoopAge)
      Maximum=&FloatFromCurve(FloatCurve=DataInterface() {propertyValues {Curve {keys [<at 0 350>,<at 0.5 700>,<at 1 75>]}}}, CurveIndex=$Emitter.NormalizedLoopAge)
     )
     Y=&RandomRangeFloat(
      Minimum=&FloatFromCurve(FloatCurve=DataInterface() {propertyValues {Curve {keys [<at 0 650>,<at 0.5 -1800>,<at 1 -1200>]}}}, CurveIndex=$Emitter.NormalizedLoopAge)
      Maximum=&FloatFromCurve(FloatCurve=DataInterface() {propertyValues {Curve {keys [<at 0 1250>,<at 0.5 -2000>,<at 1 750>]}}}, CurveIndex=$Emitter.NormalizedLoopAge)
     )
     Z=&RandomRangeFloat(Minimum=1200, Maximum=2000)
    )
    Offset Mode=::NewEnumerator0
    Override Seed=false
    Shape Primitive=::NewEnumerator1
    Spawn Group Mask=1
    Surface Only=true
    UseOwnerScale=true
   ) {module=ShapeLocation001}
   &AddVelocity(Velocity=&RandomRangeVector(Minimum=<300 150 -1200>, Maximum=<1200 -256 -2240>))
  ]
  particleUpdateScript [
   &ParticleState()
   &Collision(
    Age Colliding Particles=false
    Collision Enabled=&RandomBool()
    Friction=1
    Friction During a Bounce=1
    GPU Collision Type=::NewEnumerator2
    KillOnCollision=false
    Module.Control Rest State Via Surface Normal=false
    Module.EnableMaxCollisionCount=false
    Module.Randomize Collision Normal Vector=false
    Restitution=0
   )
   &SolveForcesAndVelocity(Clamp Velocity=false, Limit Acceleration=false)
   &ScaleSpriteSizeBySpeed(Max Scale Factor=<1 19>)
   &ScaleColor(
    Scale Alpha=&FloatFromCurve(FloatCurve=DataInterface() {propertyValues {Curve {keys [<at 0.6798 1.7048 RCIM_Cubic>,<at 1 0 RCIM_Cubic>]}}})
    ScaleA=true
    ScaleRGB=false
   )
  ]
  rendererClasses [&NiagaraSpriteRendererProperties]
  renderers [{rendererClass=&NiagaraSpriteRendererProperties}]
 }
}
summary {
 systemName=NS_Storm_Rain_Athena_Impacts
 emitters [
  {emitterName=Near_Rain_Impacts, simTarget=GPUComputeSim, rendererClasses [&NiagaraSpriteRendererProperties]}
  {emitterName=Far_Rain_Impact, simTarget=GPUComputeSim, rendererClasses [&NiagaraSpriteRendererProperties]}
  {emitterName=Distant_Rain_Sheet, rendererClasses [&NiagaraSpriteRendererProperties]}
  {
   emitterName=Lightning_Strikes
   rendererClasses [&NiagaraMeshRendererProperties, &NiagaraLightRendererProperties]
  }
 ]
 userVariables [User.StormIntensity {type=NiagaraFloat, defaultValue=0, description=""}]
}
